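; BP_WallArtProp is a Blueprint prop for wall art among the animated neon signs of the NeonCity set.
; Its construction script pushes the LumenColor and LumenBrightness variables into the Lumen Plane's materials as the "Color" and "Emissive Boost" parameters.
; The BeginPlay, overlap and Tick events are empty, so the graph adds no runtime logic.
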(event EventBeginPlay)

(event Collision|EventActorBeginOverlap
  (params (OtherActor :type Actor%20Object%20Reference)))

(event EventTick
  (params (DeltaSeconds :type Float%20%28single-precision%29)))

(fn ConstructionScript
  (bind _lumen_plane (Variables|Default|GetLumen%20Plane))
  (Rendering|Material|SetVectorParameterValueonMaterials _lumen_plane "Color" (Math|Conversions|ToVector(LinearColor) (Variables|Default|GetLumenColor)))
  (Rendering|Material|SetScalarParameterValueonMaterials _lumen_plane "Emissive Boost" (Variables|Default|GetLumenBrightness)))
