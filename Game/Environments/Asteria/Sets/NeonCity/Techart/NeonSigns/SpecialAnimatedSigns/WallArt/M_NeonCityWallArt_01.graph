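# Opaque neon wall-art material for NeonCity signs, showing animated, color-graded artwork on a glossy black LCD-style panel.
# Three scaled samples of one texture, with drifting, panning and partly distorted UVs, sum into a gradient mask.
# That mask, pulsed by a sine and offset vertically, indexes a curve atlas blended over CrystalColor.
# An LCD grid and repeating panel mask break up the result before emissive balancing; grime varies roughness, and BumpOffset nodes are unconnected.
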Material /Game/Environments/Asteria/Sets/NeonCity/Techart/NeonSigns/SpecialAnimatedSigns/WallArt/M_NeonCityWallArt_01
graph {
 properties {
  bUseMaterialAttributes=true
  bUsedWithInstancedStaticMeshes=true
  bUsedWithNanite=true
  pixelDepthOffsetMode=PDOM_Legacy
  usedShadingModels=Default Lit
 }
 nodes [
  n1 = Multiply(
   A=TextureCoordinate() {uTiling=0.5, vTiling=0.5}
   B=AppendVector(A=ScalarParameter() {defaultValue=3, parameterName=StretchX}, B=ScalarParameter() {defaultValue=1, parameterName=StretchY})
  )
  n2 = Multiply(A=Time(), B=ScalarParameter() {defaultValue=0.01, parameterName=Movement Speed})
  n3 = Sine(None=@n2)
  n4 = Constant() {r=-1}
  n5 = AppendVector(
   A=LinearInterpolate(Alpha=@n3) {constB=0.05}
   B=Multiply(A=ScalarParameter() {defaultValue=1, parameterName=PanningSpeed}, B=Multiply(A=@n2, B=@n4))
  )
  n6 = Add(A=@n1, B=@n5)
  n7 = ScalarParameter() {defaultValue=0.33, parameterName=Height Ratio}
  BumpOffset(Coordinate=@n6, Height=Constant() {r=0.1}, HeightRatioInput=@n7)
  BumpOffset(Height=Constant() {r=0.3}, HeightRatioInput=@n7)
  n8 = TextureObject() {
   samplerType=SAMPLERTYPE_LinearColor
   texture=&/WallArtGraph_M
  }
  n9 = TextureSample(UVs=Add(A=TextureCoordinate(), B=OneMinus(None=@n3)), Tex=@n8)
  n10 = LinearInterpolate(Alpha=@n3) {constB=0.01}
  n11 = Add(A=@n6, B=Add(A=@n5, B=Multiply(A=@n9.B, B=@n10) {constB=0.1}))
  BumpOffset(Coordinate=@n11, Height=Constant() {r=0.6}, HeightRatioInput=@n7)
  n12 = TextureSample(UVs=Multiply(A=@n6) {constB=0.75}, Tex=@n8) {
   samplerType=SAMPLERTYPE_LinearColor
   texture=&WallArtGraph_M
  }
  n13 = TextureSample(UVs=Multiply(A=@n11) {constB=1.5}, Tex=@n8) {
   samplerType=SAMPLERTYPE_LinearColor
   texture=&WallArtGraph_M
  }
  n14 = TextureSample(UVs=Multiply(A=@n11) {constB=2}, Tex=@n8) {
   samplerType=SAMPLERTYPE_LinearColor
   texture=&WallArtGraph_M
  }
  n15 = Add(
   A=Add(
    A=ChannelMaskParameter(None=@n12.RGB) {maskChannel=Green, parameterName=Grad Channel, defaultValue=<rgba 0 1 0 0>}.R
    B=ChannelMaskParameter(None=@n13.RGB) {maskChannel=Green, parameterName=Grad Channel, defaultValue=<rgba 0 1 0 0>}.R
   )
   B=ChannelMaskParameter(None=@n14.RGB) {maskChannel=Green, parameterName=Grad Channel, defaultValue=<rgba 0 1 0 0>}.R
  )
  n16 = VectorParameter() {parameterName=CrystalColor, defaultValue=<rgba 0.03 0 0.0125 1>}
  n17 = Time()
  n18 = ScalarParameter() {defaultValue=0.1, parameterName=PulseSpeed}
  n19 = Saturate(None=MaterialFunctionCall(Value=Multiply(A=@n17, B=@n18)) {materialFunction=&LinearSine}.Linear Sine)
  n20 = LinearInterpolate(Alpha=@n19) {constA=0.3, constB=1.28}
  n21 = LinearInterpolate(Alpha=@n19) {constA=0.6}
  n22 = MaterialFunctionCall() {materialFunction=&LinearGradient}
  n23 = CurveAtlasRowParameter(CurveTime=Add(A=Multiply(A=@n15, B=@n20), B=Multiply(A=@n21, B=@n22.VGradient))) {
   atlas=/Game/Environments/Asteria/Sets/NeonCity/Techart/Holograms/NewHologramMaterial/HologramCurves
   curve=/Game/Athena/LinearColorCurves/Curve_WallArt
   defaultValue=1
  }
  n24 = LinearInterpolate(A=LinearInterpolate(A=Multiply(A=@n16.RGB) {constB=0.5}, B=@n16.RGB, Alpha=@n9.R), B=@n23, Alpha=@n15)
  Multiply(A=@n15, B=@n24)
  n25 = TextureCoordinate()
  n26 = ScalarParameter() {defaultValue=200, parameterName=LCDGrid}
  NamedRerouteDeclaration(None=TextureSample(UVs=Multiply(A=@n25, B=@n26)) {samplerType=SAMPLERTYPE_Masks, texture=/Game/Environments/Asteria/Sets/NeonCity/Techart/Textures/T_LCD_Panels_LCDSoft}.RGB) {name=LCD pattern, nodeColor=<rgba 0.5703 1 0 1>}
  n27 = Constant()
  n28 = Multiply(A=@n24, B=NamedRerouteUsage())
  n29 = ScalarParameter() {defaultValue=0.1, parameterName=Breakup}
  n30 = SphereMask(
   A=Divide(A=Fmod(A=@n22.UGradient, B=@n29), B=@n29)
   B=Constant() {r=0.5}
   Radius=ScalarParameter() {defaultValue=0.45, parameterName=PanelThickness}
  )
  SetMaterialAttributes!(
   MaterialAttributes
   Base Color=@n27
   Metallic=@n27
   Specular=Constant() {r=0.6}
   Roughness=LinearInterpolate(
    A=ScalarParameter() {defaultValue=0.02, parameterName=Roughness}
    Alpha=TextureSample(UVs=@n1) {samplerType=SAMPLERTYPE_LinearColor, texture=/Game/Environments/Apollo/Sets/Tinfoil/Textures/T_Tinfoil_GlassGrime_M}.G
   ) {constB=0.2}
   Emissive Color=MaterialFunctionCall!(Emissive=Multiply(A=@n28, B=@n30)) {materialFunction=/Game/Environments/Asteria/Sets/NeonCity/Techart/Screen/MaterialFunctions/MF_NeonCityEmissiveBalance}.Result
  ) {attributeSetTypes [
    #1
    #2
    #3
    #4
    #5
   ]}
 ]
}
